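; Dual-wield marshmallow pickaxe cosmetic Blueprint that progressively toasts each head as its hand hits weak points.
; Impacts on the weak-point surface increment per-hand RepNotify counters, which spawn Niagara effects and step FrameSelector on dynamic materials.
; At the burnt stage, a head gains idle smoke and a flag that layers burnt swing and impact sounds.
; Weapon visibility changes reset everything, while frontend display auto-increments the counters after a short delay.
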(event EventBeginPlay
  ;[0.5688 1 0.1973 1 100] BEGINPLAY- REFS
  (bind _returnvalue (Rendering|Material|CreateDynamicMaterialInstance (Variables|Visual|GetWeaponMesh) 0 "/BRCosmetics/Weapons/FORT_Melee/Summer_Marshmallow_Female_1h/Effects/M_Pickaxe_Marshmallow.M_Pickaxe_Marshmallow" "None" :!class /Script/Engine.PrimitiveComponent))
  (bind _output_get (Variables|Default|SetM_Hand _returnvalue))
  (Rendering|Material|SetMaterial (Variables|Visual|GetWeaponMesh) 0 _output_get)
  (bind _returnvalue_2 (Rendering|Material|CreateDynamicMaterialInstance (Variables|Visual|GetWeaponMeshOffhand) 0 "/BRCosmetics/Weapons/FORT_Melee/Summer_Marshmallow_Female_1h/Effects/M_Pickaxe_Marshmallow.M_Pickaxe_Marshmallow" "None" :!class /Script/Engine.PrimitiveComponent))
  (bind _output_get_2 (Variables|Default|SetM_Offhand _returnvalue_2))
  (Rendering|Material|SetMaterial (Variables|Visual|GetWeaponMeshOffhand) 0 _output_get_2)
  (if (Reactive|Frontend|IsCosmeticObjectDisplayedinFrontendContext self)
    (Utilities|FlowControl|RetriggerableDelay 1.5)
    (Utilities|FlowControl|Sequence
      (:then_0
        _)
      (:then_1
        (CallFunction|CounterOffhand))
      (:then_2
        (if (< (Variables|Default|GetCOUNTER_Hand) 1)
          _)))
    (CallFunction|CounterHand)))

(event EventOnWeaponVisibilityChanged
  (params (bVisible :type Boolean) (bSetForLocalControllerOnly :type Boolean))
  ;[1 1 1 1 100] ON DRAW - RESET COUNTER
  (|CallParentFunction bVisible bSetForLocalControllerOnly)
  (Rendering|Material|SetScalarParameterValue (Variables|Default|GetM_Hand) "FrameSelector" 0.0 :!class /Script/Engine.MaterialInstanceDynamic)
  (Rendering|Material|SetScalarParameterValue (Variables|Default|GetM_Offhand) "FrameSelector" 0.0 :!class /Script/Engine.MaterialInstanceDynamic)
  (Variables|Default|SetCOUNTER_Hand -1)
  (Variables|Default|SetCOUNTER_Offhand -1)
  (Components|DestroyComponent (Variables|Default|GetHAND_SMOKE_REF))
  (Components|DestroyComponent (Variables|Default|GetOFFHAND_SMOKE_REF))
  (Variables|Default|SetBurnt false)
  (Variables|Default|SetBurnt%20Offhand false))

(event FortWeapon|EventOnPlayImpactFX
  (params (HitResult :type Hit%20Result%20Structure%20%28by%20ref%29) (ImpactPhysicalSurface :type EPhysicalSurface%20Enum :default "SurfaceType_Default") (SpawnedPSC :type FXSystem%20Component%20Object%20Reference))
  (bind _returnvalue (CallFunction|GetCurrentSwingState))
  ;[1 0 0.09678 1 100] PLAY VFX ON WEAKPOINTHIT
  (|CallParentFunction HitResult ImpactPhysicalSurface SpawnedPSC)
  (switch Utilities|FlowControl|Switch|SwitchonEPhysicalSurface ImpactPhysicalSurface
    (:SurfaceType4
      _)
    (:SurfaceType5
      _)
    (:SurfaceType6
      _)
    (:SurfaceType7
      _)
    (:SurfaceType12
      (switch Utilities|FlowControl|Switch|SwitchonEFortDualWieldSwingState _returnvalue
        (:MainHand
          (CallFunction|CounterHand))
        (:OffHand
          (CallFunction|CounterOffhand))))
    (:SurfaceType14
      _)
    (:SurfaceType15
      _)
    (:SurfaceType16
      _))
  (switch Utilities|FlowControl|Switch|SwitchonEFortDualWieldSwingState _returnvalue
    (:MainHand
      (CallFunction|BurntEnemyImpact))
    (:OffHand
      (CallFunction|BurntEnemyImpactOffHand))))

(event EventSwingRight
  ;[1 1 1 1 50] BURNT AUDIO SWING LAYER
  (if (Variables|Default|GetBurnt)
    (Audio|SpawnSoundAttached "/BRCosmetics/Sounds/PickAxes/SummerMarshmallowFemale/PickaxeSwing_SummerMarshmallowFemale_Burnt.PickaxeSwing_SummerMarshmallowFemale_Burnt" (Variables|Visual|GetWeaponMeshOffhand))))

(event Custom|Main%20Hand
  (bind _counter_hand (Variables|Default|GetCOUNTER_Hand))
  (bind _returnvalue (+ _counter_hand 1))
  ;[1 0 0.09678 1 100] PLAY VFX ON WEAKPOINTHIT
  ;;[1 1 1] Hand
  (if (>= (Variables|Default|GetCOUNTER_Hand) 0)
    (Utilities|FlowControl|Delay (select (Reactive|Frontend|IsCosmeticObjectDisplayedinFrontendContext self) 0.0 (Variables|Default|GetDelayBeforeVFXSpawn)))
    (Niagara|SpawnSystemAttached (Variables|Default|GetFX_WeakPoints) (Variables|Visual|GetWeaponMesh) "Root")
    (if (< (Variables|Default|GetCOUNTER_Hand) 2)
      (Rendering|Material|SetScalarParameterValue (Variables|Default|GetM_Hand) "FrameSelector" (Math|Conversions|ToFloat(Integer) _returnvalue) :!class /Script/Engine.MaterialInstanceDynamic)
      (else
        _))
    (if (== _returnvalue 2)
      (bind _returnvalue_2 (Niagara|SpawnSystemAttached "/BRCosmetics/Weapons/FORT_Melee/Summer_Marshmallow_Female_1h/Effects/NS_IdleSmokeMarshmallow.NS_IdleSmokeMarshmallow" (Variables|Visual|GetWeaponMesh) "Root"))
      (Variables|Default|SetHAND_SMOKE_REF _returnvalue_2)
      (Variables|Default|SetBurnt true))))

(event EventSwingRight2
  ;[1 1 1 1 50] BURNT AUDIO SWING LAYER
  (if (Variables|Default|GetBurnt)
    (Audio|SpawnSoundAttached "/BRCosmetics/Sounds/PickAxes/SummerMarshmallowFemale/PickaxeSwing_SummerMarshmallowFemale_Burnt.PickaxeSwing_SummerMarshmallowFemale_Burnt" (Variables|Visual|GetWeaponMeshOffhand))))

(event EventSwingLeft
  ;[1 1 1 1 50] BURNT AUDIO SWING LAYER
  (if (Variables|Default|GetBurnt%20Offhand)
    (Audio|SpawnSoundAttached "/BRCosmetics/Sounds/PickAxes/SummerMarshmallowFemale/PickaxeSwing_SummerMarshmallowFemale_Burnt.PickaxeSwing_SummerMarshmallowFemale_Burnt" (Variables|Visual|GetWeaponMeshOffhand))))

(event Custom|OffHand
  (bind _counter_offhand (Variables|Default|GetCOUNTER_Offhand))
  (bind _returnvalue (+ _counter_offhand 1))
  ;[1 0 0.09678 1 100] PLAY VFX ON WEAKPOINTHIT
  ;;[1 1 1] Offhand
  (if (>= (Variables|Default|GetCOUNTER_Offhand) 0)
    (Utilities|FlowControl|Delay (select (Reactive|Frontend|IsCosmeticObjectDisplayedinFrontendContext self) 0.0 (Variables|Default|GetDelayBeforeVFXSpawn)))
    (Niagara|SpawnSystemAttached (Variables|Default|GetFX_WeakPoints) (Variables|Visual|GetWeaponMeshOffhand) "Root")
    (if (< (Variables|Default|GetCOUNTER_Offhand) 2)
      (Rendering|Material|SetScalarParameterValue (Variables|Default|GetM_Offhand) "FrameSelector" (Math|Conversions|ToFloat(Integer) _returnvalue) :!class /Script/Engine.MaterialInstanceDynamic)
      (else
        _))
    (if (== _returnvalue 2)
      (bind _returnvalue_2 (Niagara|SpawnSystemAttached "/BRCosmetics/Weapons/FORT_Melee/Summer_Marshmallow_Female_1h/Effects/NS_IdleSmokeMarshmallow.NS_IdleSmokeMarshmallow" (Variables|Visual|GetWeaponMeshOffhand) "Root"))
      (Variables|Default|SetOFFHAND_SMOKE_REF _returnvalue_2)
      (Variables|Default|SetBurnt%20Offhand true))))

(event Custom|CounterHand
  ;[1 0 0.09678 1 100] PLAY VFX ON WEAKPOINTHIT
  (Variables|Default|SetCOUNTER_Hand (+ (Variables|Default|GetCOUNTER_Hand) 1)))

(event EventSwingLeft2
  ;[1 1 1 1 50] BURNT AUDIO SWING LAYER
  (if (Variables|Default|GetBurnt%20Offhand)
    (Audio|SpawnSoundAttached "/BRCosmetics/Sounds/PickAxes/SummerMarshmallowFemale/PickaxeSwing_SummerMarshmallowFemale_Burnt.PickaxeSwing_SummerMarshmallowFemale_Burnt" (Variables|Visual|GetWeaponMeshOffhand))))

(event Custom|CounterOffhand
  ;[1 0 0.09678 1 100] PLAY VFX ON WEAKPOINTHIT
  (Variables|Default|SetCOUNTER_Offhand (+ (Variables|Default|GetCOUNTER_Offhand) 1)))

(fn BurntEnemyImpact
  (if (Variables|Default|GetBurnt)
    (Audio|SpawnSoundAttached "/BRCosmetics/Sounds/PickAxes/SummerMarshmallowFemale/PickaxeImpactEnemy_SummerMarshmallowFemale_Burnt.PickaxeImpactEnemy_SummerMarshmallowFemale_Burnt" (Variables|Visual|GetWeaponMesh))))

(fn BurntEnemyImpact_OffHand
  (if (Variables|Default|GetBurnt%20Offhand)
    (Audio|SpawnSoundAttached "/BRCosmetics/Sounds/PickAxes/SummerMarshmallowFemale/PickaxeImpactEnemy_SummerMarshmallowFemale_Burnt.PickaxeImpactEnemy_SummerMarshmallowFemale_Burnt" (Variables|Visual|GetWeaponMesh))))

(fn OnRep_COUNTER_Hand
  (Networking|CanExecuteCosmeticEvents
    (:True
      (CallFunction|MainHand))))

(fn OnRep_COUNTER_Offhand
  (Networking|CanExecuteCosmeticEvents
    (:True
      (CallFunction|OffHand))))

(fn ConstructionScript
  (|CallParentFunction))
