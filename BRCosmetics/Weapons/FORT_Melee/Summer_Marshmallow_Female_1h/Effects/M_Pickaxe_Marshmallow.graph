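# Skeletal-mesh material for the Summer Marshmallow female one-handed pickaxe, with a vertex-color-driven selector for showing one mesh frame at a time.
# Vertex green, scaled by NumberOfFrames, is compared with rounded FrameSelector, plus a red-channel term.
# The resulting binary mask scales an offset toward the component pivot, collapsing unselected geometry.
# Diffuse, normal and mask textures plus a roughness parameter supply the Default Lit surface.
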Material /BRCosmetics/Weapons/FORT_Melee/Summer_Marshmallow_Female_1h/Effects/M_Pickaxe_Marshmallow
graph {
 properties {bUsedWithMeshDeformer=true, bUsedWithSkeletalMesh=true, pixelDepthOffsetMode=PDOM_Legacy, usedShadingModels=Default Lit}
 nodes [
  n1 = ScalarParameter() {parameterName=FrameSelector}
  n2 = Multiply(A=VertexColor().R, B=@n1)
  n3 = VertexColor()
  n4 = Subtract(A=ScalarParameter() {defaultValue=3, parameterName=NumberOfFrames})
  n5 = Subtract(A=Add(A=Multiply(A=@n3.G, B=@n4)) {constB=0.5}, B=Round(None=@n1))
  n6 = Frac(None=Saturate(None=Add(A=@n2, B=@n5)))
  Multiply(
   A=Floor(None=OneMinus(None=@n6))
   B=Subtract(
    A=MaterialFunctionCall() {materialFunction=&ComponentPivotLocation}.Result
    B=WorldPosition().XYZ
   )
  )
  ScalarParameter() {defaultValue=0.5, parameterName=Rioughness}
  TextureSample() {texture=/BRCosmetics/Weapons/FORT_Melee/Summer_Marshmallow_Female_1h/Textures/T_Summer_Marshmallow_Female_1h_D}
  TextureSample() {
   samplerType=SAMPLERTYPE_Normal
   texture=/BRCosmetics/Weapons/FORT_Melee/Summer_Marshmallow_Female_1h/Textures/T_Summer_Marshmallow_Female_1h_N
  }
  TextureSample() {
   samplerType=SAMPLERTYPE_Masks
   texture=/BRCosmetics/Weapons/FORT_Melee/Summer_Marshmallow_Female_1h/Textures/T_Summer_Marshmallow_Female_1h_S
  }
 ]
}
