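# Idle smoke effect for the marshmallow melee weapon, rising from sampled bones of the user-supplied skeletal mesh.
# The only enabled emitter, Smoke, spawns short-lived, counter-linked particles with upward velocity, curl noise and drag, drawn by sprite and ribbon renderers.
# Distance from the sampled position feeds aging and lift, while the ribbon material erodes and shifts from hot orange to pale grey.
# Three other ribbon smoke emitters are disabled.
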NiagaraSystem /BRCosmetics/Weapons/FORT_Melee/Summer_Marshmallow_Female_1h/Effects/NS_IdleSmokeMarshmallow
dependencies {
 usedDataInterfaces [
  &NiagaraDataInterfaceColorCurve
  &NiagaraDataInterfaceCurve
  &NiagaraDataInterfaceVectorCurve
  /Script/Niagara.NiagaraDataInterfaceSimpleCounter
 ]
 usedDynamicInputs [&UniformRangedFloat]
 usedModules [&SystemState]
}
emitters {
 Smoke {
  emitterUpdateScript [
   &EmitterState()
   &SpawnRate(SpawnRate=50, Use Spawn Probability=false)
  ]
  particleSpawnScript [
   &InitializeRibbon(
    Color=<rgba 10 1.75 0.2 0.3>
    Lifetime=0.5
    Module.Write Color=true
    Module.Write Lifetime=true
    Module.Write Mass=true
    Module.Write Material Random=true
    Module.Write Position=true
    Module.Write RibbonTwist=false
    Module.Write RibbonWidth=true
    Ribbon Width=&RandomRangeFloat(Minimum=1, Maximum=6)
   )
   /SetVariables(Particles.SpriteSize=&Vector2DFromFloat(Value=1)) {isSetParameters=true}
   &SkeletalMeshLocation(
    Bone Sampling Mode=::NewEnumerator4
    Inherit Velocity=true
    MaskSpawnGroup=false
    Sampled Position Offset Coordinate Space:ENiagaraMeshTransforms=::NewEnumerator3
    Skeletal Mesh=$User.Skeletal Mesh
    VelocitySpeedLimit=false
   )
   /SetVariables(Particles.RibbonLinkOrder:NiagaraFloat=&MakeFloatFromInt(Int=/Niagara/DynamicInputs/DataInterface/GenerateIntFromCounter(SimpleCounter:NiagaraDataInterfaceSimpleCounter=DataInterface() {propertyValues {GpuSyncMode=None, InitialValue=0}}))) {isSetParameters=true}
   &AddVelocity(Velocity=<0 0 1500>) {module=AddVelocity001}
  ]
  particleUpdateScript [
   &ParticleState()
   &SkeletalMeshLocation(
    Apply To Particle Position:NiagaraBool=&SetBoolByFloatComparison(A=$Particles.NormalizedAge, B=0.2, Comparison Type=::NewEnumerator6)
    Bone Sampling Mode=::NewEnumerator4
    Inherit Velocity=false
    Interpolate Mesh Sampling=false
    MaskSpawnGroup=false
    Sampled Position Offset Coordinate Space:ENiagaraMeshTransforms=::NewEnumerator3
    Skeletal Mesh=$User.Skeletal Mesh
   ) {enabled=false, module=SkeletalMeshLocation001}
   /SetVariables(
    Particles.DistanceAbove:Vector3f=&self:ClampVector(
     Input Vector:Vector3f=&Multiply_Vector(
      A=<0 0 0.1>
      B=&Subtract_Vector(A=Linked() {linkedVariable {name=Particles.Position, type=NiagaraPosition}}, B=$Output.SkeletalMeshLocation.SampledPosition)
     )
     Max:Vector3f=<0 0 2>
     Min:Vector3f=<0 0 0>
    )
    Particles.DistanceBelow:Vector3f=&self:ClampVector(
     Input Vector:Vector3f=&Multiply_Vector(
      A=<0 0 0.1>
      B=&Subtract_Vector(A=$Output.SkeletalMeshLocation.SampledPosition, B=Linked() {linkedVariable {name=Particles.Position, type=NiagaraPosition}})
     )
     Max:Vector3f=<0 0 1>
     Min:Vector3f=<0 0 0>
    )
    Particles.DistanceFromBone:NiagaraFloat=&Multiply_Float(
     A=0.01
     B=&VectorLength(Vector=&Subtract_Vector(A=Linked() {linkedVariable {name=Particles.Position, type=NiagaraPosition}}, B=$Output.SkeletalMeshLocation.SampledPosition))
    )
   ) {isSetParameters=true}
   /SetVariables(Particles.NormalizedAge:NiagaraFloat=&Lerp_Float(
     A=$Particles.NormalizedAge
     B=1
     Alpha=&ClampFloat(Float=&Add_Float(A=&MakeFloatFromVector(Vector=$Particles.DistanceAbove, Channel=::NewEnumerator2), B=-1.5))
    )) {isSetParameters=true}
   &GravityForce(Gravity=&Lerp_Vector(
     A=<0 0 700>
     B=<0 0 200>
     Alpha=&ClampFloat(Float=&Multiply_Float(A=&VectorLength(Vector=$Output.SkeletalMeshLocation.SampledVelocity), B=0.5))
    ))
   /Niagara/Modules/Update/Forces/CurlNoiseForce(
    Noise Frequency=5
    Noise Strength=&FloatFromCurve(FloatCurve=DataInterface() {propertyValues {Curve {keys [{arriveTangent=957.2371, interpMode=RCIM_Cubic, leaveTangent=957.237, tangentMode=RCTM_User, time=0, value=10}, <at 1 300>]}}})
    Pan Noise=true
    Pan Noise Field=<0 0 2>
    Random Seed:NiagaraInt32=&RandomRangeInt(Minimum=0, Maximum=100)
   )
   &Drag(
    Drag=&Multiply_Float(
     A=&ClampFloat(Float=/Niagara/DynamicInputs/TypeConversions/VectorLengthSafe(Vector:Vector3f=$Particles.Velocity), Max=5000)
     B=&Lerp_Float(A=0.5, B=0, Alpha=$Particles.NormalizedAge)
    )
    Use Linear Drag=true
    Use Rotational Drag=false
   )
   &AddVelocity(Velocity=&Multiply_VectorByFloat(Vector=$Particles.DistanceBelow, Float=&RandomRangeFloat(Minimum=0, Maximum=100))) {module=AddVelocity003}
   &AddVelocity(
    Coordinate Space=::World
    Velocity=&Lerp_Vector(
     A=<0 0 0>
     B=<0 0 100>
     Alpha=&FloatFromCurve(
      FloatCurve=DataInterface() {propertyValues {Curve {keys [{interpMode=RCIM_Cubic, tangentMode=RCTM_User, time=0.2, value=0}, <at 1 1>]}}}
      CurveIndex=$Particles.DistanceFromBone
      Scale Curve=&RandomRangeFloat(Minimum=0.25, Maximum=1)
     )
    )
   ) {enabled=false}
   /Game/Effects/Niagara/Modules/Materials/FN_RibbonMaterial/FN_RibbonMaterial(
    Color:LinearColor=&ColorFromCurve(ColorCurve=DataInterface() {propertyValues {
       AlphaCurve {keys [<at 0 1 RCIM_Cubic>]}
       BlueCurve {keys [<at 0 0.4902 RCIM_Cubic>,<at 0.3146 0.918 RCIM_Cubic>]}
       GreenCurve {keys [<at 0 1.3333 RCIM_Cubic>,<at 0.3146 0.8632 RCIM_Cubic>]}
       RedCurve {keys [<at 0 5 RCIM_Cubic>,<at 0.3146 0.8528 RCIM_Cubic>]}
      }})
    Color Pan U Speed=-2.5345
    Color Scale U=7.5
    Erosion Amount:NiagaraFloat=&Lerp_Float(A=0.5, B=1.6, Alpha=$Particles.NormalizedAge)
    Erosion Pan Speed U=-0.8346
    Erossion Scale U=4.5
    Macro Distortion Amount=1
    Opacity Pan U Speed=-1.6786
    Opacity Power:NiagaraFloat=&RandomRangeFloat(Minimum=0.3, Maximum=1)
   ) {module=FN_RibbonMaterial002}
   &ScaleRibbonWidth(Ribbon Width Scale=&FloatFromCurve(
     FloatCurve=DataInterface() {propertyValues {Curve {keys [
         {arriveTangent=4.0699, interpMode=RCIM_Cubic, leaveTangent=4.0699, tangentMode=RCTM_User, time=0, value=0}
         <at 0.5 0.9666 RCIM_Cubic>
         {arriveTangent=-2.5115, interpMode=RCIM_Cubic, leaveTangent=-2.5115, tangentMode=RCTM_User, time=1, value=0}
        ]}}}
     Scale Curve=&RandomRangeFloat(Minimum=0.5, Maximum=2)
    )) {module=ScaleRibbonWidth001}
   &SolveForcesAndVelocity(Clamp Velocity=false, Limit Acceleration=false)
   &Color(Color=&ColorFromCurve(
     ColorCurve=DataInterface() {propertyValues {
       AlphaCurve {keys [<at 0 1 RCIM_Cubic>]}
       BlueCurve {keys [<at 0 0 RCIM_Cubic>,<at 0.4 0>,<at 0.6 0>,<at 1 1 RCIM_Cubic>]}
       GreenCurve {keys [<at 0 0.0148 RCIM_Cubic>,<at 0.4 0.8005>,<at 0.6 1>,<at 1 0.087 RCIM_Cubic>]}
       RedCurve {keys [<at 0 1 RCIM_Cubic>,<at 0.4 1>,<at 0.6 0.0681>,<at 1 0 RCIM_Cubic>]}
      }}
     CurveIndex=$Particles.DistanceFromBone
    )) {enabled=false}
  ]
  rendererClasses [&NiagaraSpriteRendererProperties, &NiagaraRibbonRendererProperties]
  renderers [
   {rendererClass=&NiagaraSpriteRendererProperties}
   {rendererClass=&NiagaraRibbonRendererProperties, rendererIndex=1}
  ]
 }
 SmokeRibbon {
  bEnabled=false
  emitterUpdateScript [
   &EmitterState()
   /Game/Effects/Niagara/Modules/Skydiving/FN_DivingAttributes/FN_EmitterDivingAttributes_v1(Accumulate Dive=-0.0025, Accumulate Rest=0, SkeletalMesh:NiagaraDataInterfaceSkeletalMesh=$User.Skeletal Mesh) {enabled=false}
   &FN_SpawnRibbon_v2(Spawn Mode=::NewEnumerator0, SpawnRate=50)
  ]
  particleSpawnScript [
   &FN_SetSpawnGroup_v2(Number of Spawn Groups=1)
   &FN_InitializeRibbon_v3(
    Color:LinearColor=&MultiplyLinearColor(Linear Color=<rgba 0.05 0.05 0.05 1>, Scale Factor=<1 1 1 1>)
    Custom Offset=true
    Enable Skeletal Position=true
    Facing Axis:ENiagaraOrientationAxis=::YAxis
    Offset Amount=1
    Ribbon Width=10
   )
   &SystemLocation(MaskSpawnGroup=false) {enabled=false}
   /SetVariables(Particles.RibbonNumber:NiagaraInt32=$Particles.SpawnGroup) {isSetParameters=true, module=SetVariables_#1}
   /SetVariables(Particles.Color=&LerpLinearColors(
     StartColor=&ColorFromCurve(
      ColorCurve=DataInterface() {propertyValues {
        AlphaCurve {keys [<at 0 1 RCIM_Cubic>]}
        BlueCurve {keys [<at 0 1 RCIM_Cubic>,<at 0.7704 0.651>,<at 1 1 RCIM_Cubic>]}
        GreenCurve {keys [<at 0 0.7687 RCIM_Cubic>,<at 0.7704 0.1163>,<at 1 0.0833 RCIM_Cubic>]}
        RedCurve {keys [<at 0 0.0525 RCIM_Cubic>,<at 0.7704 0>,<at 1 0.5317 RCIM_Cubic>]}
       }}
      CurveIndex=&RandomRangeFloat(Minimum=0, Maximum=1)
     )
     EndColor=&ColorFromCurve(
      ColorCurve=DataInterface() {propertyValues {
        AlphaCurve {keys [<at 0 1 RCIM_Cubic>]}
        BlueCurve {keys [<at 0 1 RCIM_Cubic>,<at 0.8279 0.651>,<at 1 1 RCIM_Cubic>]}
        GreenCurve {keys [<at 0 0.7687 RCIM_Cubic>,<at 0.8279 0.1163>,<at 1 0.0833 RCIM_Cubic>]}
        RedCurve {keys [<at 0 0.0525 RCIM_Cubic>,<at 0.8279 0>,<at 1 0.5317 RCIM_Cubic>]}
       }}
      CurveIndex=&RandomRangeFloat(Minimum=0, Maximum=1)
     )
     LerpFactor=&MakeFloatFromInt(Int=$Particles.SpawnGroup)
    )) {enabled=false, isSetParameters=true, module=SetVariables_#2}
   &Color(Color=$Particles.Color, Scale Color=&RandomRangeVector(Minimum=<2 1 1>, Maximum=<10 1 1>)) {enabled=false}
   &AddVelocity(Velocity=<300 0 300>)
   /SetVariables(Particles.SmokeColor:LinearColor=<rgba 0 0 0 0.1>) {enabled=false, isSetParameters=true, module=SetVariables_#3}
   &SampleSkeletalMeshSkeleton(Mesh:NiagaraDataInterfaceSkeletalMesh=$User.Skeletal Mesh) {enabled=false}
   &SkeletalMeshLocation(
    Interpolate Mesh Sampling=false
    MaskSpawnGroup=false
    Mesh Sampling Type=::NewEnumerator5
    Sampled Position Offset=<-5 0 35>
    Skeletal Mesh=$User.Skeletal Mesh
    Socket Sampling Mode=::NewEnumerator5
   )
   &SphereLocation(
    MaskSpawnGroup=false
    Override Seed=false
    Sphere Radius=0
    Surface Only=false
    UseNonUniformScale=false
    UseOffset=false
    UseSphereOrientation=false
   ) {enabled=false}
  ]
  particleUpdateScript [
   &ScaleVelocity(Velocity Scale=&MakeVector(
     X=&FloatFromCurve(FloatCurve=DataInterface() {propertyValues {Curve {keys [<at 0 1 RCIM_Cubic>,<at 0.3 0 RCIM_Cubic>]}}})
     Y=&FloatFromCurve(FloatCurve=DataInterface() {propertyValues {Curve {keys [<at 0 1 RCIM_Cubic>,<at 0.3 0 RCIM_Cubic>]}}})
     Z=1
    ))
   &DynamicMaterialParameters(Module.Param0WriteEnabled=true, Module.Param1WriteEnabled=true, Module.Param2WriteEnabled=true, Module.Param3WriteEnabled=true) {module=DynamicMaterialParameters001}
   /Niagara/Modules/Update/Forces/V2/CurlNoiseForce(
    Curl Noise Cone Mask Angle=20
    Curl Noise Cone Mask Falloff Angle=20
    Mask Curl Noise=true
    Noise Frequency=4
    Noise Strength=80000
    Pan Noise=false
   ) {enabled=false}
   &SolveForcesAndVelocity(Clamp Velocity=false, Limit Acceleration=false)
   &FN_UpdateRibbon(
    Enable Skeletal Position=true
    Initial Color:LinearColor=$Particles.Initial.Color
    Initial RibbonWidth:NiagaraFloat=$Particles.Initial.RibbonWidth
   )
   &ScaleColor(
    Scale Alpha=&FloatFromCurve(FloatCurve=DataInterface() {propertyValues {Curve {keys [<at 0 1>, {arriveTangent=-0.1164, interpMode=RCIM_Cubic, leaveTangent=-0.1164, tangentMode=RCTM_User, time=0.4, value=1}, <at 1 0>]}}})
    Scale RGB=&VectorFromCurve(VectorCurve=DataInterface() {propertyValues {
       XCurve {keys [<at 0.0618 1>,<at 0.1318 1>,<at 0.25 1>,<at 0.8 0.9>]}
       YCurve {keys [<at 0.25 1>,<at 0.75 1>]}
       ZCurve {keys [<at 0.0618 1>,<at 0.1318 1>,<at 0.25 1>,<at 0.8 0.9>]}
      }})
    ScaleA=true
    ScaleRGB=true
   )
   &ScaleColor(
    Scale RGB=&VectorFromFloat(Value=&FloatFromCurve(FloatCurve=DataInterface() {propertyValues {Curve {keys [
          {arriveTangent=0.1804, interpMode=RCIM_Cubic, leaveTangent=0.1804, tangentMode=RCTM_User, time=0, value=6}
          {arriveTangent=-39.5437, interpMode=RCIM_Cubic, leaveTangent=-39.5437, tangentMode=RCTM_User, time=0.3, value=1}
         ]}}}))
    ScaleA=true
    ScaleRGB=true
   ) {module=ScaleColor001}
   &ScaleRibbonWidth(Ribbon Width Scale=&FloatFromCurve(FloatCurve=DataInterface() {propertyValues {Curve {keys [<at 0 0>,<at 0.2 1>]}}}, Scale Curve=4))
  ]
  rendererClasses [&NiagaraRibbonRendererProperties]
  renderers [{rendererClass=&NiagaraRibbonRendererProperties}]
 }
 SmokeRibbon_WORLD {
  bEnabled=false
  emitterUpdateScript [
   &EmitterState()
   &FN_EmitterDivingAttributes_v1(Accumulate Dive=-0.0025, Accumulate Rest=0, SkeletalMesh:NiagaraDataInterfaceSkeletalMesh=$User.Skeletal Mesh) {enabled=false}
   &FN_SpawnRibbon_v2(Spawn Mode=::NewEnumerator0, SpawnRate=50)
  ]
  particleSpawnScript [
   &FN_SetSpawnGroup_v2(Number of Spawn Groups=1)
   &FN_InitializeRibbon_v3(
    Color:LinearColor=&MultiplyLinearColor(Linear Color=<rgba 0.05 0.05 0.05 1>, Scale Factor=<1 1 1 1>)
    Custom Offset=true
    Enable Skeletal Position=true
    Facing Axis:ENiagaraOrientationAxis=::YAxis
    Lifetime=1
    Offset Amount=1
    Ribbon Width=10
   )
   &SystemLocation(MaskSpawnGroup=false) {enabled=false}
   /SetVariables(Particles.RibbonNumber:NiagaraInt32=$Particles.SpawnGroup) {isSetParameters=true, module=SetVariables_#1}
   /SetVariables(Particles.Color=&LerpLinearColors(
     StartColor=&ColorFromCurve(
      ColorCurve=DataInterface() {propertyValues {
        AlphaCurve {keys [<at 0 1 RCIM_Cubic>]}
        BlueCurve {keys [<at 0 1 RCIM_Cubic>,<at 0.7704 0.651>,<at 1 1 RCIM_Cubic>]}
        GreenCurve {keys [<at 0 0.7687 RCIM_Cubic>,<at 0.7704 0.1163>,<at 1 0.0833 RCIM_Cubic>]}
        RedCurve {keys [<at 0 0.0525 RCIM_Cubic>,<at 0.7704 0>,<at 1 0.5317 RCIM_Cubic>]}
       }}
      CurveIndex=&RandomRangeFloat(Minimum=0, Maximum=1)
     )
     EndColor=&ColorFromCurve(
      ColorCurve=DataInterface() {propertyValues {
        AlphaCurve {keys [<at 0 1 RCIM_Cubic>]}
        BlueCurve {keys [<at 0 1 RCIM_Cubic>,<at 0.8279 0.651>,<at 1 1 RCIM_Cubic>]}
        GreenCurve {keys [<at 0 0.7687 RCIM_Cubic>,<at 0.8279 0.1163>,<at 1 0.0833 RCIM_Cubic>]}
        RedCurve {keys [<at 0 0.0525 RCIM_Cubic>,<at 0.8279 0>,<at 1 0.5317 RCIM_Cubic>]}
       }}
      CurveIndex=&RandomRangeFloat(Minimum=0, Maximum=1)
     )
     LerpFactor=&MakeFloatFromInt(Int=$Particles.SpawnGroup)
    )) {enabled=false, isSetParameters=true, module=SetVariables_#2}
   &Color(Color=$Particles.Color, Scale Color=&RandomRangeVector(Minimum=<2 1 1>, Maximum=<10 1 1>)) {enabled=false}
   &AddVelocity(Coordinate Space=::Simulation, Velocity=<0 0 500>)
   /SetVariables(Particles.SmokeColor:LinearColor=<rgba 0 0 0 0.1>) {enabled=false, isSetParameters=true, module=SetVariables_#3}
   &SampleSkeletalMeshSkeleton(Mesh:NiagaraDataInterfaceSkeletalMesh=$User.Skeletal Mesh) {enabled=false}
   &SkeletalMeshLocation(
    Bone Sampling Mode=::NewEnumerator4
    Interpolate Mesh Sampling=false
    MaskSpawnGroup=false
    Sampled Position Offset=<-10 0 0>
    Sampled Position Offset Coordinate Space:ENiagaraSkelMeshTransforms=::NewEnumerator4
    Skeletal Mesh=$User.Skeletal Mesh
   )
   &SphereLocation(
    MaskSpawnGroup=false
    Override Seed=false
    Sphere Radius=0
    Surface Only=false
    UseNonUniformScale=false
    UseOffset=false
    UseSphereOrientation=false
   ) {enabled=false}
  ]
  particleUpdateScript [
   &ScaleVelocity(Velocity Scale=&MakeVector(
     X=&FloatFromCurve(FloatCurve=DataInterface() {propertyValues {Curve {keys [<at 0 1 RCIM_Cubic>,<at 0.3 0 RCIM_Cubic>]}}})
     Y=&FloatFromCurve(FloatCurve=DataInterface() {propertyValues {Curve {keys [<at 0 1 RCIM_Cubic>,<at 0.3 0 RCIM_Cubic>]}}})
     Z=1
    ))
   &InheritSourceMovement(
    Applied Force Scale=<1 1 0>
    Applied Position Scale=<0 0 1>
    ApplyForce=true
    ApplyForceFalloff=false
    ApplyPosition=true
    ApplyPositionFalloff=false
    Module.LimitForceSpeed=false
   ) {enabled=false}
   &DynamicMaterialParameters(Module.Param0WriteEnabled=true, Module.Param1WriteEnabled=true, Module.Param2WriteEnabled=true, Module.Param3WriteEnabled=true) {module=DynamicMaterialParameters001}
   &CurlNoiseForce(
    Curl Noise Cone Mask Angle=20
    Curl Noise Cone Mask Falloff Angle=20
    Mask Curl Noise=true
    Noise Frequency=4
    Noise Strength=80000
    Pan Noise=false
   ) {enabled=false}
   &SolveForcesAndVelocity(Clamp Velocity=false, Limit Acceleration=false)
   &FN_UpdateRibbon(
    Enable Skeletal Position=true
    Initial Color:LinearColor=$Particles.Initial.Color
    Initial RibbonWidth:NiagaraFloat=$Particles.Initial.RibbonWidth
   )
   &ScaleColor(
    Scale Alpha=&FloatFromCurve(FloatCurve=DataInterface() {propertyValues {Curve {keys [<at 0 1>, {arriveTangent=-0.1164, interpMode=RCIM_Cubic, leaveTangent=-0.1164, tangentMode=RCTM_User, time=0.4, value=1}, <at 1 0>]}}})
    Scale RGB=&VectorFromCurve(VectorCurve=DataInterface() {propertyValues {
       XCurve {keys [<at 0.0618 1>,<at 0.1318 1>,<at 0.25 1>,<at 0.8 0.9>]}
       YCurve {keys [<at 0.25 1>,<at 0.75 1>]}
       ZCurve {keys [<at 0.0618 1>,<at 0.1318 1>,<at 0.25 1>,<at 0.8 0.9>]}
      }})
    ScaleA=true
    ScaleRGB=true
   )
   &ScaleColor(
    Scale RGB=&VectorFromFloat(Value=&FloatFromCurve(FloatCurve=DataInterface() {propertyValues {Curve {keys [
          {arriveTangent=0.1804, interpMode=RCIM_Cubic, leaveTangent=0.1804, tangentMode=RCTM_User, time=0, value=6}
          {arriveTangent=-39.5437, interpMode=RCIM_Cubic, leaveTangent=-39.5437, tangentMode=RCTM_User, time=0.3, value=1}
         ]}}}))
    ScaleA=true
    ScaleRGB=true
   ) {module=ScaleColor001}
   &ScaleRibbonWidth(Ribbon Width Scale=&FloatFromCurve(FloatCurve=DataInterface() {propertyValues {Curve {keys [<at 0 0>,<at 0.2 1>]}}}, Scale Curve=4))
  ]
  rendererClasses [&NiagaraRibbonRendererProperties]
  renderers [{rendererClass=&NiagaraRibbonRendererProperties}]
 }
 SmokeTrail {
  bEnabled=false
  emitterUpdateScript [
   &EmitterState()
   /Game/Effects/Niagara/Modules/Ribbons/FN_SpawnRibbon_v3(
    Spawn Mode:ENiagaraSkeletalSpawning=::NewEnumerator1
    SpawnRate:NiagaraFloat=&Multiply_Float(
     A=20
     B=&MakeCustomFloatFromBool(Bool=&Boolean_OrOperation(Bool1=$User.IsSkydiving, Bool2=$User.IsDriving), True Float=0, False Float=1)
    )
   )
  ]
  particleSpawnScript [
   &InitializeRibbon(Color=<rgba 1 1 1 40>, Lifetime=0.45, Ribbon Width=20, UsePositionOffset=false, Write Material Random=true)
   &SkeletalMeshLocation(
    Bone Sampling Mode=::NewEnumerator4
    Interpolate Mesh Sampling=false
    MaskSpawnGroup=false
    Sampled Position Offset=<-10 -3 -4>
    Sampled Position Offset Coordinate Space:ENiagaraSkelMeshTransforms=::NewEnumerator2
    Skeletal Mesh=$User.Skeletal Mesh
   ) {module=SkeletalMeshLocation001}
   &SkeletalMeshLocation(
    Bone Sampling Mode=::NewEnumerator4
    Interpolate Mesh Sampling=false
    MaskSpawnGroup=false
    Sampled Position Offset=<-5 0 0>
    Skeletal Mesh=$User.Skeletal Mesh
   ) {enabled=false}
   &AddVelocity(Velocity=<200 0 100>)
  ]
  particleUpdateScript [
   &InheritSourceMovement(
    Applied Force Scale=<-4 -4 1>
    ApplyForce=true
    ApplyForceFalloff=false
    ApplyPosition=true
    ApplyPositionFalloff=false
    Module.LimitForceSpeed=false
   )
   &CurlNoiseForce(Noise Frequency=500, Noise Strength=150, Pan Noise=false) {enabled=false}
   &SolveForcesAndVelocity(Clamp Velocity=false, Limit Acceleration=false)
   &ParticleState()
   /Game/Effects/Niagara/Modules/Materials/FN_RibbonMaterial/FN_Ribbon_v3/FN_RibbonMaterial(
    Color=<rgba 0.2135 0.2135 0.2135 1>
    Erosion Amount=0.75
    Erosion Pan Speed U=0.2
    Erosion Power=4
    Erossion Scale U=0.5
    Macro Distortion Amount=0.2
    Macro Pan Speed=0.7
    Macro Scale U=2
    Micro Distortion Amount=0.05
    Micro Pan Speed=0.2
    Micro Scale U=0.2
    Opacity Erosion=0.1
    Opacity Pan U Speed=0.15
    Scale Alpha=0.8
    UV0 Offset:NiagaraFloat=&RandomRangeFloat(Minimum=0, Maximum=1)
   ) {module=FN_RibbonMaterial001}
   &ScaleSpriteSize(Uniform Curve Sprite Scale=DataInterface() {propertyValues {Curve {keys [{tangentMode=RCTM_User, time=0, value=1}, {tangentMode=RCTM_User, time=1, value=0}]}}})
   &ScaleRibbonWidth(Ribbon Width Scale=&FloatFromCurve(FloatCurve=DataInterface() {propertyValues {Curve {keys [<at 0 1>,<at 0.5 0>]}}}, Scale Curve=1.3)) {enabled=false}
  ]
  rendererClasses [&NiagaraRibbonRendererProperties]
  renderers [{rendererClass=&NiagaraRibbonRendererProperties}]
 }
}
summary {
 systemName=NS_IdleSmokeMarshmallow
 emitters [
  {bEnabled=false, emitterName=SmokeRibbon, rendererClasses [&NiagaraRibbonRendererProperties]}
  {bEnabled=false, emitterName=SmokeRibbon_WORLD, rendererClasses [&NiagaraRibbonRendererProperties]}
  {bEnabled=false, emitterName=SmokeTrail, rendererClasses [&NiagaraRibbonRendererProperties]}
  {emitterName=Smoke, rendererClasses [&NiagaraSpriteRendererProperties, &NiagaraRibbonRendererProperties]}
 ]
 userVariables [
  User.bIsFrontEnd {type=NiagaraBool}
  User.bIsFrontEndPreview {type=NiagaraBool}
  User.IsDriving {type=NiagaraBool}
  User.IsSkydiving {type=NiagaraBool}
  User.RibbonIDOffset {type=NiagaraInt32}
  User.Skeletal Mesh {
   type=NiagaraDataInterfaceSkeletalMesh
   defaultValue {
    struct=NiagaraExt_VariableValue_DataInterface
    value {dataInterface=&self:NiagaraDataInterfaceSkeletalMesh_0}
   }
  }
 ]
}
